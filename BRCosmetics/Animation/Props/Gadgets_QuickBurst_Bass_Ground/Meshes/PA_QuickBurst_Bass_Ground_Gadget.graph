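# Physics asset for the QuickBurst Bass Ground gadget prop, giving its skeleton capsule collision bodies.
# A very wide, near-spherical capsule sits on root, and a smaller, offset and rotated capsule covers JNT_Ground.
# The offset and main bones carry only tiny placeholder-sized capsules.
# Constraints chain root, offset, main and JNT_Ground in sequence, each with the same swing and twist angle limits.
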PhysicsAsset /BRCosmetics/Animation/Props/Gadgets_QuickBurst_Bass_Ground/Meshes/PA_QuickBurst_Bass_Ground_Gadget
bodies {
 JNT_Ground [{shapeType=Capsule, length=86.7544, radius=74.022, center=<6.085 10.4954 -18.6928>, rotation=<rot 0 143.7503 -86.4433>}]
 main [{shapeType=Capsule, length=0.505, radius=0.505, rotation=<rot 0 -90 0>}]
 offset [{shapeType=Capsule, length=0.505, radius=0.505, rotation=<rot 0 -90 0>}]
 root [{shapeType=Capsule, length=0.2651, radius=186.72, rotation=<rot 0 -90 0>}]
}
constraints [
 {bone1Name=offset, bone2Name=root, swing1LimitDegrees=45, swing2LimitDegrees=45, twistLimitDegrees=45}
 {bone1Name=main, bone2Name=offset, swing1LimitDegrees=45, swing2LimitDegrees=45, twistLimitDegrees=45}
 {bone1Name=JNT_Ground, bone2Name=main, swing1LimitDegrees=45, swing2LimitDegrees=45, twistLimitDegrees=45}
]
